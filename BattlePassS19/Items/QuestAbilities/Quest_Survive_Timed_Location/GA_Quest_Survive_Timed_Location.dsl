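; Gameplay ability for a timed-location survival quest that credits the player while their pawn is tagged as in a target location.
; On activation it caches the pawn, ability system and quest manager, then checks immediately and on a looping one-second timer.
; Each check sends a complex custom stat event with the configured tag if the pawn has any Locations tag.
; Otherwise, it clears the timer and ends the ability.
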(event Ability|EventFailedToActivatePassiveAbility
  (params (ActorInfo :type Gameplay%20Ability%20Actor%20Info%20Structure)))

(event Ability|EventActivateAbility
  (bind _returnvalue (CallFunction|GetAvatarasFortPlayerPawn))
  (bind _returnvalue_2 (Ability|GetFortAbilitySystemComponent _returnvalue))
  (|CallParentFunction)
  (Utilities|IsValid _returnvalue
    (:Is%20Valid
      (Utilities|IsValid _returnvalue_2
        (:Is%20Valid
          (Variables|Default|SetOwnerAbilitySystem _returnvalue_2)
          (bind _output_get (Variables|GAVars|Data|SetOwningPlayerPawn _returnvalue))
          (Utilities|IsValid (Quests|GetQuestManager (Fortnite|GetFortPlayerControllerfromActor _output_get) "Invalid")
            (:Is%20Valid
              (Variables|Default|SetOwnerQuestManager (Quests|GetQuestManager (Fortnite|GetFortPlayerControllerfromActor _output_get) "Invalid"))
              (CallFunction|LoopCheckinLocation)
              (bind _returnvalue_3 (Utilities|Time|SetTimerbyEvent (delegate Custom|LoopCheckInLocation) 1.0 true))
              (Variables|Default|SetLoopTimer _returnvalue_3))))))))

(event Custom|LoopCheckInLocation
  (if (GameplayTags|HasAnyMatchingGameplayTags (Variables|GAVars|Data|GetOwningPlayerPawn) (Variables|Default|GetLocations))
    (Quest|SendComplexCustomStatEvent (Variables|Default|GetOwnerQuestManager) 0 (GameplayTags|MakeGameplayTagContainerfromTag (Variables|Default|GetComplexCustomTag)) "(GameplayTags=)" 1)
    (else
      (Utilities|Time|ClearandInvalidateTimerbyHandle (Variables|Default|GetLoopTimer))
      (Ability|EndAbility))))
